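# Floating interaction indicator for a spatial frontend navigation point, built on FortEventLevelNavigationWidget, showing a donut pip that becomes a "Computer" select prompt on hover.
# Hover start plays a sound and glitchy reveal of the input icon and label, then loops a ring pulse when the animation finishes.
# Hover end hides the prompt and replays the pip's pop-in.
# SetCATextFX copies the label into offset red and blue layers for chromatic aberration.
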WidgetBlueprint /BRCosmetics/Athena/UI/Frontend/Spatial/Art/InteractionIndicators/FloatingInteraction_Indicator_DonutMonitor
animations [
 PopluateIn {
  binding_count=1
  length_seconds=0.25
  object=&self:PopluateIn
  sequence {
   bindings [FloatingPip {
     id=#1
     possessed_class=Image
     tracks [
      {
       type=MovieSceneWidgetMaterialTrack
       sections [{
         type=MovieSceneParameterSection
         channels [Overall_Alpha {type=MovieSceneScriptingFloatChannel, keys [[0,0],[5,0],[8,0.5],[10,0],[13,0],[16,0.25],[18,0],[25,1]]}]
        }]
      }
      {
       type=MovieScene2DTransformTrack
       sections [{
         type=MovieScene2DTransformSection
         channels [
          Translation.X {type=MovieSceneScriptingFloatChannel}
          Translation.Y {type=MovieSceneScriptingFloatChannel, keys [[5,8],[20,0]]}
          Angle {type=MovieSceneScriptingFloatChannel}
          Scale.X {type=MovieSceneScriptingFloatChannel, keys [[5,0.75],[25,1]]}
          Scale.Y {type=MovieSceneScriptingFloatChannel, keys [[5,0.75],[25,1]]}
          Shear.X {type=MovieSceneScriptingFloatChannel}
          Shear.Y {type=MovieSceneScriptingFloatChannel}
         ]
        }]
      }
      {
       type=MovieSceneFloatTrack
       display_name=Render Opacity
       sections [{type=MovieSceneFloatSection, channels [None {type=MovieSceneScriptingFloatChannel, keys [[0,0],[2,1]]}]}]
      }
     ]
    }]
   display_rate {numerator=100}
  }
 }
 Hover {
  binding_count=11
  length_seconds=1
  object=&self:Hover
  sequence {
   bindings [
    Divider {
     id=#2
     possessed_class=Image
     tracks [{
       type=MovieScene2DTransformTrack
       sections [{
         type=MovieScene2DTransformSection
         channels [
          Translation.X {type=MovieSceneScriptingFloatChannel, keys [[0,20],[4.8,20],[7.8,24],[12.001,0]]}
          Translation.Y {type=MovieSceneScriptingFloatChannel}
          Angle {type=MovieSceneScriptingFloatChannel}
          Scale.X {type=MovieSceneScriptingFloatChannel}
          Scale.Y {type=MovieSceneScriptingFloatChannel, thinned_from=8, keys [[0,0],[4.8,0],[7.8,0.5],[15.6,0.5],[18,0.75],[21,1.1],[24,1]]}
          Shear.X {type=MovieSceneScriptingFloatChannel}
          Shear.Y {type=MovieSceneScriptingFloatChannel}
         ]
        }]
      }]
    }
    InputAction_Overlay {
     id=#3
     possessed_class=Overlay
     tracks [
      {
       type=MovieSceneFloatTrack
       display_name=Render Opacity
       sections [{type=MovieSceneFloatSection, channels [None {type=MovieSceneScriptingFloatChannel, keys [[19.2,0],[22.2,1]]}]}]
      }
      {
       type=MovieScene2DTransformTrack
       sections [{
         type=MovieScene2DTransformSection
         channels [
          Translation.X {type=MovieSceneScriptingFloatChannel}
          Translation.Y {type=MovieSceneScriptingFloatChannel}
          Angle {type=MovieSceneScriptingFloatChannel}
          Scale.X {type=MovieSceneScriptingFloatChannel, keys [[19.2,0],[25.201,1.1],[28.201,1]]}
          Scale.Y {type=MovieSceneScriptingFloatChannel, keys [[19.2,0],[25.201,1.1],[28.201,1]]}
          Shear.X {type=MovieSceneScriptingFloatChannel}
          Shear.Y {type=MovieSceneScriptingFloatChannel}
         ]
        }]
      }
     ]
    }
    InputActionMask {
     id=#4
     possessed_class=Image
     tracks [{
       type=MovieSceneWidgetMaterialTrack
       sections [{type=MovieSceneParameterSection, channels [Alpha {type=MovieSceneScriptingFloatChannel, keys [[24.001,1],[28.201,1],[31.2,0]]}]}]
      }]
    }
    OverlayLeft {
     id=#5
     possessed_class=Image
     tracks [
      {
       type=MovieSceneFloatTrack
       display_name=Render Opacity
       sections [{type=MovieSceneFloatSection, channels [None {type=MovieSceneScriptingFloatChannel, keys [[15,0],[24,1],[26.999,1],[29.999,0]]}]}]
      }
      {
       type=MovieScene2DTransformTrack
       sections [{
         type=MovieScene2DTransformSection
         channels [
          Translation.X {type=MovieSceneScriptingFloatChannel}
          Translation.Y {type=MovieSceneScriptingFloatChannel}
          Angle {type=MovieSceneScriptingFloatChannel}
          Scale.X {type=MovieSceneScriptingFloatChannel, keys [[15.001,0],[24,1]]}
          Scale.Y {type=MovieSceneScriptingFloatChannel}
          Shear.X {type=MovieSceneScriptingFloatChannel}
          Shear.Y {type=MovieSceneScriptingFloatChannel}
         ]
        }]
      }
     ]
    }
    DigitalPattern {
     id=#6
     possessed_class=Image
     tracks [{
       type=MovieSceneFloatTrack
       display_name=Render Opacity
       sections [{type=MovieSceneFloatSection, channels [None {type=MovieSceneScriptingFloatChannel, keys [[0,0],[15,0],[45,1]]}]}]
      }]
    }
    Select {
     id=#7
     possessed_class=CommonTextBlock
     tracks [
      {
       type=MovieSceneWidgetMaterialTrack
       display_name=Font.FontMaterial
       sections [{type=MovieSceneParameterSection, channels [FIllWhite {type=MovieSceneScriptingFloatChannel, keys [[1.8,1],[3,0],[21,0],[30,1]]}]}]
      }
      {
       type=MovieSceneWidgetMaterialTrack
       display_name=Font.Outline Settings.OutlineMaterial
       sections [{type=MovieSceneParameterSection, channels [Alpha {type=MovieSceneScriptingFloatChannel, keys [[1.8,0],[3,1],[24,1],[54,0]]}]}]
      }
      {
       type=MovieSceneFloatTrack
       display_name=Render Opacity
       sections [{type=MovieSceneFloatSection, channels [None {type=MovieSceneScriptingFloatChannel, keys [[1,0],[2.4,1]]}]}]
      }
      {
       type=MovieSceneColorTrack
       display_name=Shadow Color
       sections [{
         type=MovieSceneColorSection
         channels [
          Color.R {type=MovieSceneScriptingFloatChannel, thinned_from=8, keys [[1.8,0],[7.2,0],[7.8,1],[10.2,1],[10.8,0],[16.2,0]]}
          Color.G {type=MovieSceneScriptingFloatChannel, thinned_from=9, keys [[1.8,1],[7.2,1],[7.8,0],[10.2,0],[10.8,1],[16.2,1]]}
          Color.B {type=MovieSceneScriptingFloatChannel, thinned_from=9, keys [[1.8,1],[10.2,1],[10.8,0],[16.2,0],[16.8,1]]}
          Color.A {type=MovieSceneScriptingFloatChannel, keys [[1.8,0.5],[19.2,0.5],[19.8,0]]}
         ]
        }]
      }
      {
       type=MovieSceneFloatVectorTrack
       display_name=Shadow Offset
       sections [{
         type=MovieSceneFloatVectorSection
         channels [
          Vector.X {
           type=MovieSceneScriptingFloatChannel
           keys [[1.8,0],[4.2,0],[4.8,-1.3117],[7.2,-1.3117],[7.8,4],[10.2,4],[10.8,2.466],[13.2,2.466],[13.8,3.5],[16.2,3.5],[16.8,-1]]
          }
          Vector.Y {
           type=MovieSceneScriptingFloatChannel
           keys [[1.8,0],[4.2,0],[4.8,1.5568],[7.2,1.5568],[7.8,1],[10.2,1],[10.8,-2.8145],[13.2,4],[13.8,0.0115],[16.2,0.0115],[16.8,-1]]
          }
         ]
        }]
      }
      {
       type=MovieScene2DTransformTrack
       sections [{
         type=MovieScene2DTransformSection
         channels [
          Translation.X {type=MovieSceneScriptingFloatChannel}
          Translation.Y {type=MovieSceneScriptingFloatChannel, const [0,1]}
          Angle {type=MovieSceneScriptingFloatChannel, const [0,1]}
          Scale.X {type=MovieSceneScriptingFloatChannel, keys [[0,0.6],[1.8,0.75],[2.4,0.6],[3.6,0.6],[4.8,1.1],[5.4,0.95],[6,1]]}
          Scale.Y {type=MovieSceneScriptingFloatChannel, keys [[0,0.5],[1.8,0.4],[2.4,1],[3,1],[3.6,0.5],[6,1]]}
          Shear.X {type=MovieSceneScriptingFloatChannel, keys [[0,16],[1.8,16],[2.4,-8],[3.6,-8],[4.2,24],[5.4,24],[6,-12],[6.6,0]]}
          Shear.Y {type=MovieSceneScriptingFloatChannel, const [0,1]}
         ]
        }]
      }
     ]
    }
    FloatingPip {
     id=#8
     possessed_class=Image
     tracks [{
       type=MovieSceneFloatTrack
       display_name=Render Opacity
       sections [{type=MovieSceneFloatSection, channels [None {type=MovieSceneScriptingFloatChannel, keys [[0,1],[3,0]]}]}]
      }]
    }
    SelectState {
     id=#9
     possessed_class=Border
     tracks [{
       type=MovieSceneFloatTrack
       display_name=Render Opacity
       sections [{type=MovieSceneFloatSection, channels [None {type=MovieSceneScriptingFloatChannel, keys [[0,0.25],[3,1]]}]}]
      }]
    }
    SelectionStateBox {
     id=#10
     possessed_class=HorizontalBox
     tracks [
      {
       type=MovieSceneFloatTrack
       display_name=Render Opacity
       sections [{type=MovieSceneFloatSection, channels [None {type=MovieSceneScriptingFloatChannel, keys [[0,0],[1.2,1]]}]}]
      }
      {
       type=MovieScene2DTransformTrack
       sections [{
         type=MovieScene2DTransformSection
         channels [
          Translation.X {type=MovieSceneScriptingFloatChannel, keys [[0,40],[9,0]]}
          Translation.Y {type=MovieSceneScriptingFloatChannel}
          Angle {type=MovieSceneScriptingFloatChannel}
          Scale.X {type=MovieSceneScriptingFloatChannel}
          Scale.Y {type=MovieSceneScriptingFloatChannel}
          Shear.X {type=MovieSceneScriptingFloatChannel}
          Shear.Y {type=MovieSceneScriptingFloatChannel}
         ]
        }]
      }
     ]
    }
    CA_TextFX {
     id=#11
     possessed_class=Overlay
     tracks [{
       type=MovieSceneFloatTrack
       display_name=Render Opacity
       sections [{type=MovieSceneFloatSection, start_frame=0, channels [None {type=MovieSceneScriptingFloatChannel, keys [[0,0],[15.001,0],[60,1]]}]}]
      }]
    }
    OverlayRight {
     id=#12
     possessed_class=Image
     tracks [
      {
       type=MovieSceneFloatTrack
       display_name=Render Opacity
       sections [{
         type=MovieSceneFloatSection
         channels [None {type=MovieSceneScriptingFloatChannel, keys [[24.001,0],[27,1],[48,1],[49.2,0],[51,0],[54,1],[55.8,0],[58.8,0],[60,1]]}]
        }]
      }
      {
       type=MovieScene2DTransformTrack
       sections [{
         type=MovieScene2DTransformSection
         channels [
          Translation.X {type=MovieSceneScriptingFloatChannel}
          Translation.Y {type=MovieSceneScriptingFloatChannel}
          Angle {type=MovieSceneScriptingFloatChannel}
          Scale.X {type=MovieSceneScriptingFloatChannel, keys [[24.001,1],[30,1],[39,0.025],[51,0.025]]}
          Scale.Y {type=MovieSceneScriptingFloatChannel, keys [[39,1],[51,1],[57,0]]}
          Shear.X {type=MovieSceneScriptingFloatChannel}
          Shear.Y {type=MovieSceneScriptingFloatChannel}
         ]
        }]
      }
     ]
    }
   ]
   display_rate {numerator=60}
  }
 }
 UnHover {
  binding_count=1
  length_seconds=0.05
  object=&self:UnHover
  sequence {bindings [SelectionStateBox {
     id=#13
     possessed_class=HorizontalBox
     tracks [{
       type=MovieSceneFloatTrack
       display_name=Render Opacity
       sections [{type=MovieSceneFloatSection, channels [None {type=MovieSceneScriptingFloatChannel, keys [[0,1],[1,0]]}]}]
      }]
    }]}
 }
 LoopingHoverState {
  binding_count=5
  length_seconds=3.5
  object=&self:LoopingHoverState
  sequence {bindings [
    InputAction_Overlay {
     id=#14
     possessed_class=Overlay
     tracks [
      {
       type=MovieSceneFloatTrack
       display_name=Render Opacity
       sections [{type=MovieSceneFloatSection, channels [None {type=MovieSceneScriptingFloatChannel, const [1,1]}]}]
      }
      {
       type=MovieScene2DTransformTrack
       sections [{
         type=MovieScene2DTransformSection
         channels [
          Translation.X {type=MovieSceneScriptingFloatChannel}
          Translation.Y {type=MovieSceneScriptingFloatChannel}
          Angle {type=MovieSceneScriptingFloatChannel}
          Scale.X {type=MovieSceneScriptingFloatChannel, const [1,1]}
          Scale.Y {type=MovieSceneScriptingFloatChannel, const [1,1]}
          Shear.X {type=MovieSceneScriptingFloatChannel}
          Shear.Y {type=MovieSceneScriptingFloatChannel}
         ]
        }]
      }
     ]
    }
    InputActionRing {
     id=#15
     possessed_class=Image
     tracks [{
       type=MovieSceneFloatTrack
       display_name=Render Opacity
       sections [{type=MovieSceneFloatSection, channels [None {type=MovieSceneScriptingFloatChannel, keys [[10,0],[12,1],[20,1],[29.2,0]]}]}]
      }]
    }
    Divider {
     id=#16
     possessed_class=Image
     tracks [{
       type=MovieScene2DTransformTrack
       sections [{
         type=MovieScene2DTransformSection
         channels [
          Translation.X {type=MovieSceneScriptingFloatChannel, const [0,1]}
          Translation.Y {type=MovieSceneScriptingFloatChannel}
          Angle {type=MovieSceneScriptingFloatChannel}
          Scale.X {type=MovieSceneScriptingFloatChannel}
          Scale.Y {type=MovieSceneScriptingFloatChannel, const [1,1]}
          Shear.X {type=MovieSceneScriptingFloatChannel}
          Shear.Y {type=MovieSceneScriptingFloatChannel}
         ]
        }]
      }]
    }
    InputActionMask {
     id=#17
     possessed_class=Image
     tracks [{
       type=MovieSceneWidgetMaterialTrack
       sections [{type=MovieSceneParameterSection, channels [Alpha {type=MovieSceneScriptingFloatChannel, const [0,1]}]}]
      }]
    }
    CircleBurst {
     id=#18
     possessed_class=Image
     tracks [
      {
       type=MovieSceneWidgetMaterialTrack
       sections [{type=MovieSceneParameterSection, channels [Alpha {type=MovieSceneScriptingFloatChannel, keys [[12,0],[14,0.25],[32,0],[70,0]]}]}]
      }
      {
       type=MovieScene2DTransformTrack
       sections [{
         type=MovieScene2DTransformSection
         channels [
          Translation.X {type=MovieSceneScriptingFloatChannel}
          Translation.Y {type=MovieSceneScriptingFloatChannel}
          Angle {type=MovieSceneScriptingFloatChannel}
          Scale.X {type=MovieSceneScriptingFloatChannel, keys [[12,1],[32,3],[70,3]]}
          Scale.Y {type=MovieSceneScriptingFloatChannel, keys [[12,1],[32,3],[70,3]]}
          Shear.X {type=MovieSceneScriptingFloatChannel}
          Shear.Y {type=MovieSceneScriptingFloatChannel}
         ]
        }]
      }
      {
       type=MovieSceneFloatTrack
       display_name=Render Opacity
       sections [{type=MovieSceneFloatSection, start_frame=0, channels [None {type=MovieSceneScriptingFloatChannel, keys [[10,0],[12,1]]}]}]
      }
     ]
    }
   ]}
 }
]
graphs [
 EventGraph {nodes [
   n1 = TriggerLoopingHoverStateAnim() {is=title K2Node_CustomEvent}
   > PlayAnimation(InAnimation=Get LoopingHoverState() -> (LoopingHoverState:Widget Animation Object Reference), NumLoopsToPlay=0)
   n2 = Event PreConstruct()
   > n3 = Branch(Condition=@n2.IsDesignTime)
   SetCATextFX(execute=[@n4.then, @n3.then])
   Event Construct()
   > SetRenderOpacity(self=Get SelectionStateBox() -> (SelectionStateBox:Horizontal Box Object Reference), InOpacity=0)
   > n4 = SetRenderOpacity(self=Get FloatingPip() -> (FloatingPip:Image Object Reference), InOpacity=1)
   Event OnObjectHoverBegin!() -> (OutputDelegate, then, NavWidgetSettings:"Nav Widget Settings Structure (by ref)") {is=title K2Node_Event}
   > StopAnimation(InAnimation=Get UnHover() -> (UnHover:Widget Animation Object Reference))
   > PlayAnimation(InAnimation=Get Hover() -> (Hover:Widget Animation Object Reference), PlaybackSpeed=1.5)
   > PlaySound2D(Sound=/BRCosmetics/Sounds/Other/Fort_UI_Sounds/S12_Frontend/UI_Faction_Intel_Text_Cue)
   > BindToAnimationFinished(Animation=Get Hover(), Delegate=@n1)
   Event OnObjectHoverEnd() {is=title K2Node_Event}
   > StopAnimation(InAnimation=Get Hover())
   > StopAnimation(InAnimation=Get LoopingHoverState())
   > SetRenderOpacity(self=Get SelectionStateBox(), InOpacity=0)
   n5 = Get PopluateIn() -> (PopluateIn:Widget Animation Object Reference)
   > PlayAnimationForward(InAnimation=@n5)
   PlayAnimation(InAnimation=@n5)
   Event Tick()
  ]}
 SetCATextFX {nodes [
   SetCATextFX() {is=title K2Node_FunctionEntry}
   > "SetText (Text)"(
    self=[Get Title_red() -> (Title_red:Common Text Object Reference), Get Title_Blue() -> (Title_Blue:Common Text Object Reference)]
    InText="GetText (Text)"!(self:Text Object Reference=Get Select() -> (Select:Common Text Object Reference)) -> (ReturnValue:Text)
   )
  ]}
]
variables []
widgets {
 info {inheritedWidgetCount=1, parentClass=/Script/FortniteGame.FortEventLevelNavigationWidget, rootWidgetClass=&Overlay, widgetCount=24}
 widgets [
  Overlay_18 {widgetClassPath=&Overlay, widget=Overlay_18}
  FloatingPip {
   widgetClassPath=&Image
   bIsVariable=true
   parent=Overlay_18
   widget=FloatingPip
   properties {brush {
     bIsDynamicallyLoaded=false
     drawAs=Image
     imageType=NoImage
     mirroring=NoMirror
     resourceName=None
     resourceObject=/FortniteCoreUI/Frontend/Spatial/Art/InteractionIndicators/M_UI_Floating_POI_Pip_Donut
     tiling=NoTile
     imageSize=<85 85>
     margin=<ltrb 0 0 0 0>
     outlineSettings {
      bUseBrushTransparency=false
      roundingType=HalfHeightRadius
      width=0
      color {colorUseRule=UseColor_Specified, specifiedColor=<rgba 0 0 0 0>}
      cornerRadii=<0 0 0 0>
     }
     tintColor {colorUseRule=UseColor_Specified, specifiedColor=<rgba 1 1 1 1>}
     uVRegion {bIsValid=false, max=<0 0>, min=<0 0>}
    }}
   slot {class=OverlaySlot, properties {horizontalAlignment=HAlign_Center, verticalAlignment=VAlign_Center}}
  }
  SelectionStateBox {
   widgetClassPath=&HorizontalBox
   bIsVariable=true
   parent=Overlay_18
   widget=SelectionStateBox
   properties {renderTransform {angle=0, scale=<1 1>, shear=<-3 -3>, translation=<0 0>}}
   slot {class=OverlaySlot, properties {horizontalAlignment=HAlign_Fill, verticalAlignment=VAlign_Fill}}
  }
  InputAction_Overlay {
   widgetClassPath=&Overlay
   parent=SelectionStateBox
   widget=InputAction_Overlay
   slot {class=HorizontalBoxSlot, properties {horizontalAlignment=HAlign_Center, verticalAlignment=VAlign_Center}}
  }
  CircleBurst {
   widgetClassPath=&Image
   bIsVariable=true
   parent=InputAction_Overlay
   widget=CircleBurst
   properties {
    renderOpacity=0
    brush {
     bIsDynamicallyLoaded=false
     drawAs=Image
     imageType=NoImage
     mirroring=NoMirror
     resourceName=None
     resourceObject=/FortniteCoreUI/Frontend/Spatial/Art/InteractionIndicators/M_UI_Circle_Master_GlitchExlore
     tiling=NoTile
     imageSize=<28 28>
     margin=<ltrb 0 0 0 0>
     outlineSettings {
      bUseBrushTransparency=false
      roundingType=HalfHeightRadius
      width=0
      color {colorUseRule=UseColor_Specified, specifiedColor=<rgba 0 0 0 0>}
      cornerRadii=<0 0 0 0>
     }
     tintColor {colorUseRule=UseColor_Specified, specifiedColor=<rgba 1 1 1 1>}
     uVRegion {bIsValid=false, max=<0 0>, min=<0 0>}
    }
   }
   slot {class=OverlaySlot, properties {horizontalAlignment=HAlign_Center, verticalAlignment=VAlign_Center}}
  }
  Arrow {
   widgetClassPath=&Image
   bIsVariable=true
   parent=InputAction_Overlay
   widget=Arrow
   properties {brush {
     bIsDynamicallyLoaded=false
     drawAs=Image
     imageType=NoImage
     mirroring=NoMirror
     resourceName=None
     resourceObject=/FortniteCoreUI/Frontend/Spatial/Art/Elevator/M_UI_Elevator_ContextState_Icon_Proceed
     tiling=NoTile
     imageSize=<28 28>
     margin=<ltrb 0 0 0 0>
     outlineSettings {
      bUseBrushTransparency=false
      roundingType=HalfHeightRadius
      width=0
      color {colorUseRule=UseColor_Specified, specifiedColor=<rgba 0 0 0 0>}
      cornerRadii=<0 0 0 0>
     }
     tintColor {colorUseRule=UseColor_Specified, specifiedColor=<rgba 1 1 1 1>}
     uVRegion {bIsValid=false, max=<0 0>, min=<0 0>}
    }}
   slot {class=OverlaySlot, properties {horizontalAlignment=HAlign_Center, verticalAlignment=VAlign_Center}}
  }
  InputActionRing {
   widgetClassPath=&Image
   bIsVariable=true
   parent=InputAction_Overlay
   widget=InputActionRing
   properties {
    renderOpacity=0
    brush {
     bIsDynamicallyLoaded=false
     drawAs=Image
     imageType=NoImage
     mirroring=NoMirror
     resourceName=None
     resourceObject=&M_UI_Circle_Master_GlitchExlore
     tiling=NoTile
     imageSize=<28 28>
     margin=<ltrb 0 0 0 0>
     outlineSettings {
      bUseBrushTransparency=false
      roundingType=HalfHeightRadius
      width=0
      color {colorUseRule=UseColor_Specified, specifiedColor=<rgba 0 0 0 0>}
      cornerRadii=<0 0 0 0>
     }
     tintColor {colorUseRule=UseColor_Specified, specifiedColor=<rgba 1 1 1 1>}
     uVRegion {bIsValid=false, max=<0 0>, min=<0 0>}
    }
    colorAndOpacity=<rgba 1 1 1 0.75>
    renderTransform {angle=0, scale=<1.333 1.333>, shear=<0 0>, translation=<0 0>}
   }
   slot {class=OverlaySlot, properties {horizontalAlignment=HAlign_Center, verticalAlignment=VAlign_Center}}
  }
  SizeBox_1 {
   widgetClassPath=&SizeBox
   parent=InputAction_Overlay
   widget=SizeBox_1
   properties {bOverride_HeightOverride=true, bOverride_WidthOverride=true, heightOverride=28, widthOverride=28}
   slot {class=OverlaySlot, properties {horizontalAlignment=HAlign_Center, verticalAlignment=VAlign_Center}}
  }
  InputActionWidget {
   widgetClassPath=&CommonActionWidget
   bIsVariable=true
   parent=SizeBox_1
   widget=InputActionWidget
   properties {
    visibility=SelfHitTestInvisible
    iconRimBrush {
     bIsDynamicallyLoaded=false
     drawAs=NoDrawType
     imageType=NoImage
     mirroring=NoMirror
     resourceName=None
     resourceObject=null
     tiling=NoTile
     imageSize=<28 28>
     margin=<ltrb 0 0 0 0>
     outlineSettings {
      bUseBrushTransparency=false
      roundingType=HalfHeightRadius
      width=0
      color {colorUseRule=UseColor_Specified, specifiedColor=<rgba 0 0 0 0>}
      cornerRadii=<0 0 0 0>
     }
     tintColor {colorUseRule=UseColor_Specified, specifiedColor=<rgba 1 1 1 1>}
     uVRegion {bIsValid=false, max=<0 0>, min=<0 0>}
    }
    inputActions [{dataTable=&InputActionDataTable, rowName=Select}]
   }
   slot {class=SizeBoxSlot}
  }
  InputActionMask {
   widgetClassPath=&Image
   bIsVariable=true
   parent=InputAction_Overlay
   widget=InputActionMask
   properties {
    brush {
     bIsDynamicallyLoaded=false
     drawAs=Image
     imageType=NoImage
     mirroring=NoMirror
     resourceName=None
     resourceObject=/FortniteCoreUI/Frontend/Spatial/Art/InteractionIndicators/M_UI_Circle_Master_GlitchExlore_Inst
     tiling=NoTile
     imageSize=<28 28>
     margin=<ltrb 0 0 0 0>
     outlineSettings {
      bUseBrushTransparency=false
      roundingType=HalfHeightRadius
      width=0
      color {colorUseRule=UseColor_Specified, specifiedColor=<rgba 0 0 0 0>}
      cornerRadii=<0 0 0 0>
     }
     tintColor {colorUseRule=UseColor_Specified, specifiedColor=<rgba 1 1 1 1>}
     uVRegion {bIsValid=false, max=<0 0>, min=<0 0>}
    }
    renderTransform {angle=0, scale=<1.05 1.05>, shear=<0 0>, translation=<0 0>}
   }
   slot {class=OverlaySlot, properties {horizontalAlignment=HAlign_Center, verticalAlignment=VAlign_Center}}
  }
  SelectState {
   widgetClassPath=&Border
   parent=SelectionStateBox
   widget=SelectState
   properties {
    verticalAlignment=VAlign_Center
    background {
     bIsDynamicallyLoaded=false
     drawAs=NoDrawType
     imageType=NoImage
     mirroring=NoMirror
     resourceName=None
     resourceObject=null
     tiling=NoTile
     imageSize=<32 32>
     margin=<ltrb 0 0 0 0>
     outlineSettings {
      bUseBrushTransparency=false
      roundingType=HalfHeightRadius
      width=0
      color {colorUseRule=UseColor_Specified, specifiedColor=<rgba 0 0 0 0>}
      cornerRadii=<0 0 0 0>
     }
     tintColor {colorUseRule=UseColor_Specified, specifiedColor=<rgba 1 1 1 1>}
     uVRegion {bIsValid=false, max=<0 0>, min=<0 0>}
    }
    padding=<ltrb 0 0 0 0>
   }
   slot {class=HorizontalBoxSlot, properties {padding=<ltrb 6 0 0 0>}}
  }
  HorizontalBox_55 {
   widgetClassPath=&HorizontalBox
   parent=SelectState
   widget=HorizontalBox_55
   slot {class=BorderSlot, properties {verticalAlignment=VAlign_Center, padding=<ltrb 0 0 0 0>}}
  }
  Overlay_3 {
   widgetClassPath=&Overlay
   parent=HorizontalBox_55
   widget=Overlay_3
   slot {class=HorizontalBoxSlot, properties {verticalAlignment=VAlign_Center}}
  }
  DigitalPattern {
   widgetClassPath=&Image
   bIsVariable=true
   parent=Overlay_3
   widget=DigitalPattern
   properties {
    brush {
     bIsDynamicallyLoaded=false
     drawAs=Box
     imageType=NoImage
     mirroring=NoMirror
     resourceName=None
     resourceObject=/FortniteCoreUI/Frontend/Spatial/Art/InteractionIndicators/M_UI_POI_Glitch_Text_Backing
     tiling=NoTile
     imageSize=<64 48>
     margin=<ltrb 0.5 0 0.5 0>
     outlineSettings {
      bUseBrushTransparency=false
      roundingType=HalfHeightRadius
      width=0
      color {colorUseRule=UseColor_Specified, specifiedColor=<rgba 0 0 0 0>}
      cornerRadii=<0 0 0 0>
     }
     tintColor {colorUseRule=UseColor_Specified, specifiedColor=<rgba 1 1 1 1>}
     uVRegion {bIsValid=false, max=<0 0>, min=<0 0>}
    }
    renderTransform {angle=0, scale=<1.2 1.25>, shear=<0 0>, translation=<-4 0>}
   }
   slot {class=OverlaySlot, properties {horizontalAlignment=HAlign_Fill, verticalAlignment=VAlign_Fill}}
  }
  Overlay_41 {
   widgetClassPath=&Overlay
   parent=Overlay_3
   widget=Overlay_41
   slot {class=OverlaySlot, properties {horizontalAlignment=HAlign_Fill, verticalAlignment=VAlign_Center}}
  }
  OverlayLeft {
   widgetClassPath=&Image
   bIsVariable=true
   parent=Overlay_41
   widget=OverlayLeft
   properties {
    renderOpacity=0
    brush {
     bIsDynamicallyLoaded=false
     drawAs=Image
     imageType=NoImage
     mirroring=NoMirror
     resourceName=None
     resourceObject=null
     tiling=NoTile
     imageSize=<4 28>
     margin=<ltrb 0 0 0 0>
     outlineSettings {
      bUseBrushTransparency=false
      roundingType=HalfHeightRadius
      width=0
      color {colorUseRule=UseColor_Specified, specifiedColor=<rgba 0 0 0 0>}
      cornerRadii=<0 0 0 0>
     }
     tintColor {colorUseRule=UseColor_Specified, specifiedColor=<rgba 0 0.6667 1 1>}
     uVRegion {bIsValid=false, max=<0 0>, min=<0 0>}
    }
    colorAndOpacity=<rgba 1 1 1 0.333>
    renderTransformPivot=<0 0.5>
   }
   slot {class=OverlaySlot, properties {horizontalAlignment=HAlign_Fill, verticalAlignment=VAlign_Center}}
  }
  OverlayRight {
   widgetClassPath=&Image
   bIsVariable=true
   parent=Overlay_41
   widget=OverlayRight
   properties {
    renderOpacity=0
    brush {
     bIsDynamicallyLoaded=false
     drawAs=Image
     imageType=NoImage
     mirroring=NoMirror
     resourceName=None
     resourceObject=null
     tiling=NoTile
     imageSize=<4 28>
     margin=<ltrb 0 0 0 0>
     outlineSettings {
      bUseBrushTransparency=false
      roundingType=HalfHeightRadius
      width=0
      color {colorUseRule=UseColor_Specified, specifiedColor=<rgba 0 0 0 0>}
      cornerRadii=<0 0 0 0>
     }
     tintColor {colorUseRule=UseColor_Specified, specifiedColor=<rgba 0 0.6667 1 1>}
     uVRegion {bIsValid=false, max=<0 0>, min=<0 0>}
    }
    colorAndOpacity=<rgba 1 1 1 0.5>
    renderTransformPivot=<1 0.5>
   }
   slot {class=OverlaySlot, properties {horizontalAlignment=HAlign_Fill, verticalAlignment=VAlign_Center}}
  }
  Divider {
   widgetClassPath=&Image
   bIsVariable=true
   parent=Overlay_41
   widget=Divider
   properties {brush {
     bIsDynamicallyLoaded=false
     drawAs=Image
     imageType=NoImage
     mirroring=NoMirror
     resourceName=None
     resourceObject=&/M_UI_InteractionIndicator_DividerLine_Inst
     tiling=NoTile
     imageSize=<4 28>
     margin=<ltrb 0 0 0 0>
     outlineSettings {
      bUseBrushTransparency=false
      roundingType=HalfHeightRadius
      width=0
      color {colorUseRule=UseColor_Specified, specifiedColor=<rgba 0 0 0 0>}
      cornerRadii=<0 0 0 0>
     }
     tintColor {colorUseRule=UseColor_Specified, specifiedColor=<rgba 0 0.6667 1 1>}
     uVRegion {bIsValid=false, max=<0 0>, min=<0 0>}
    }}
   slot {class=OverlaySlot, properties {verticalAlignment=VAlign_Center}}
  }
  Overlay_2 {
   widgetClassPath=&Overlay
   parent=Overlay_3
   widget=Overlay_2
   slot {class=OverlaySlot, properties {horizontalAlignment=HAlign_Fill, verticalAlignment=VAlign_Center, padding=<ltrb 4 0 0 0>}}
  }
  CA_TextFX {
   widgetClassPath=&Overlay
   parent=Overlay_2
   widget=CA_TextFX
   slot {class=OverlaySlot, properties {horizontalAlignment=HAlign_Fill}}
  }
  Title_Blue {
   widgetClassPath=&CommonTextBlock
   bIsVariable=true
   parent=CA_TextFX
   widget=Title_Blue
   properties {
    justification=Center
    textTransformPolicy=ToUpper
    colorAndOpacity {colorUseRule=UseColor_Specified, specifiedColor=<rgba 0.102 0.9403 1 0.5>}
    font {
     bForceMonospaced=false
     bMaterialIsStencil=false
     fontMaterial=/FortniteCoreUI/Frontend/Spatial/Art/GlitchFX/M_UI_GlitchText_Fill_LowNoise
     fontObject=&Burbank
     letterSpacing=0
     monospacedWidth=1
     size=24
     skewAmount=0
     typefaceFontName=BurbankRegular-Black
     outlineSettings {
      bApplyOutlineToDropShadows=false
      bMiteredCorners=false
      bSeparateFillAlpha=false
      outlineMaterial=/FortniteCoreUI/Frontend/Spatial/Art/GlitchFX/M_UI_GlitchText_Outline
      outlineSize=1
      outlineColor=<rgba 0.1022 0.9473 1 0.5>
     }
    }
    margin=<ltrb 4 2 8 0>
    renderTransform {angle=0, scale=<1 1>, shear=<0 0>, translation=<2 2>}
    shadowOffset=<0 0>
   }
   slot {class=OverlaySlot, properties {horizontalAlignment=HAlign_Fill, verticalAlignment=VAlign_Center}}
  }
  Title_red {
   widgetClassPath=&CommonTextBlock
   bIsVariable=true
   parent=CA_TextFX
   widget=Title_red
   properties {
    justification=Center
    textTransformPolicy=ToUpper
    colorAndOpacity {colorUseRule=UseColor_Specified, specifiedColor=<rgba 1 0.107 0.0802 0.333>}
    font {
     bForceMonospaced=false
     bMaterialIsStencil=false
     fontMaterial=&M_UI_GlitchText_Fill_LowNoise
     fontObject=&Burbank
     letterSpacing=0
     monospacedWidth=1
     size=24
     skewAmount=0
     typefaceFontName=BurbankRegular-Black
     outlineSettings {
      bApplyOutlineToDropShadows=false
      bMiteredCorners=false
      bSeparateFillAlpha=false
      outlineMaterial=&M_UI_GlitchText_Outline
      outlineSize=1
      outlineColor=<rgba 1 0.107 0.0802 0.5>
     }
    }
    margin=<ltrb 4 2 8 0>
    renderTransform {angle=0, scale=<1 1>, shear=<0 0>, translation=<-1 1>}
    shadowOffset=<0 0>
   }
   slot {class=OverlaySlot, properties {horizontalAlignment=HAlign_Fill, verticalAlignment=VAlign_Center}}
  }
  Select {
   widgetClassPath=&CommonTextBlock
   bIsVariable=true
   parent=Overlay_2
   widget=Select
   properties {
    justification=Center
    text=Computer
    textTransformPolicy=ToUpper
    font {
     bForceMonospaced=false
     bMaterialIsStencil=false
     fontMaterial=/FortniteCoreUI/Frontend/Spatial/Art/GlitchFX/M_UI_GlitchText_Fill
     fontObject=&Burbank
     letterSpacing=0
     monospacedWidth=1
     size=24
     skewAmount=0
     typefaceFontName=BurbankRegular-Black
     outlineSettings {
      bApplyOutlineToDropShadows=false
      bMiteredCorners=false
      bSeparateFillAlpha=false
      outlineMaterial=&M_UI_GlitchText_Outline
      outlineSize=2
      outlineColor=<rgba 1 1 1 1>
     }
    }
    margin=<ltrb 4 2 8 0>
    shadowColorAndOpacity=<rgba 0 1 1 0>
    shadowOffset=<0 0>
   }
   slot {class=OverlaySlot, properties {horizontalAlignment=HAlign_Fill, verticalAlignment=VAlign_Center}}
  }
  Button_Navigation {
   widgetClassPath=/Game/Athena/UI/Matchmaking/SimpleCommonButton
   bInherited=true
   bIsVariable=true
   parent=Overlay_18
   widget=Button_Navigation
   properties {
    bConsumePointerInput=true
    onButtonBaseClicked=(&self:WidgetTree.Button_Navigation.HandleOnButtonClicked)
    onButtonBaseDoubleClicked=(&self:WidgetTree.Button_Navigation.HandleOnButtonDoubleClicked)
    onButtonBaseHovered=(&self:WidgetTree.Button_Navigation.HandleOnButtonHovered)
    onButtonBaseUnhovered=(&self:WidgetTree.Button_Navigation.HandleOnButtonUnhovered)
    onSelectedChangedBase=(&self:WidgetTree.Button_Navigation.HandleOnSelectedChanged)
    style=/FortUILibrary/Foundation/Buttons/ButtonStyles/Primary/ButtonStyle-Primary-M_NoHoverSound.ButtonStyle-Primary-M_NoHoverSound_C
    colorAndOpacity=<rgba 1 1 1 0>
   }
   slot {class=OverlaySlot, properties {horizontalAlignment=HAlign_Fill, verticalAlignment=VAlign_Fill}}
  }
 ]
}
